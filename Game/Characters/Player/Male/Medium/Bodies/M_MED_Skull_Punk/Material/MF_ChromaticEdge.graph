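# Material function for the Skull Punk player body that adds an animated chromatic rim glow to the incoming emissive attribute.
# A curve atlas supplies the color, indexed by time, mask offset and a view-space normal arctangent term with Fresnel-weighted distortion.
# The arctangent takes the normal's G component for both inputs, so it reflects only that component's sign.
# Fresnel, effect mask, AO input and an intensity parameter scale the glow.
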MaterialFunction /Game/Characters/Player/Male/Medium/Bodies/M_MED_Skull_Punk/Material/MF_ChromaticEdge
graph {nodes [
  n1 = GetMaterialAttributes(None=FunctionInput() {inputType=FunctionInput_MaterialAttributes}) {attributeGetTypes [#1]}
  n2 = FunctionInput() {bUsePreviewValueAsDefault=true, inputName=AO, inputType=FunctionInput_Scalar, sortPriority=200, previewValue=<1 0 0 1>}
  n3 = TextureSample(Tex=FunctionInput() {inputName=EffectMask, inputType=FunctionInput_Texture2D, sortPriority=100}) {samplerSource=SSM_Wrap_WorldGroupSettings}
  n4 = Add(A=@n3.R, B=@n3.R)
  n5 = PixelNormalWS()
  n6 = MaterialFunctionCall(Float2=MaterialFunctionCall(Vector=Transform(None=@n5) {transformSourceType=TRANSFORMSOURCE_World, transformType=TRANSFORM_View}) {materialFunction=&SafeNormalize}.Result) {materialFunction=&BreakOutFloat2Components}
  n7 = ConstantBiasScale(None=Multiply(A=Arctangent2Fast(Y=@n6.G, X=@n6.G)) {constB=0.3183, desc=Divide by 3.14159})
  n8 = Multiply(A=@n3.R) {constB=0.5}
  n9 = Time()
  n10 = Add(
   A=Multiply(A=ScalarParameter() {defaultValue=3, group=ChromaticEdge, parameterName=ChromaticEdge_Tile}, B=@n7)
   B=Multiply(A=Add(A=@n8, B=@n9), B=ScalarParameter() {defaultValue=0.066, group=ChromaticEdge, parameterName=ChromaticEdge_Speed})
  )
  n11 = Fresnel(ExponentIn=ScalarParameter() {defaultValue=5, group=ChromaticEdge, parameterName=ChromaticEdge_Exp}) {baseReflectFraction=0}
  n12 = Multiply(A=Multiply(A=@n7, B=ScalarParameter() {defaultValue=0.2, group=ChromaticEdge, parameterName=ChromaticEdge_NormalDistort}), B=@n11)
  n13 = CurveAtlasRowParameter(CurveTime=Frac(None=Add(A=@n10, B=@n12))) {
   atlas=/Game/Characters/Player/Male/Medium/Bodies/M_MED_Skull_Punk/Textures/SkullPunkCurveAtlas
   curve=/Game/Characters/Player/Male/Medium/Bodies/M_MED_Skull_Punk/Textures/SkullPunk_EdgeColorCurve
   group=ChromaticEdge
   parameterName=ChromaticEdgeCurve
  }
  n14 = Multiply(
   A=Multiply(A=@n2, B=Saturate(None=@n4))
   B=Multiply(A=Multiply(A=@n13, B=@n11), B=ScalarParameter() {defaultValue=0.7, group=ChromaticEdge, parameterName=ChromaticEdge_Intensity})
  )
  FunctionOutput(None=SetMaterialAttributes!(MaterialAttributes=@n1.MaterialAttributes, EmissiveColor=Add(A=@n1.EmissiveColor, B=@n14)) {attributeSetTypes [#1]})
 ]}
